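# Material function for UI particle sparks, outputting a procedural mask of animated streaks from rotatable, tiled UVs.
# An integer-hash Custom node turns each grid cell plus a seed into a random value declared as a named reroute.
# Time-driven sine-cosine wobble and wrapping vertical travel offset each cell's UVs, which are also scaled along the pattern's Y gradient.
# A line-segment distance Custom node draws each spark, softened by an inverted SmoothStep.
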MaterialFunction /Game/UI/Foundation/Environment/Shell/Materials/MF_UI_ParticleSparks
graph {nodes [
  n1 = FunctionInput(Preview=Constant()) {bUsePreviewValueAsDefault=true, inputName=SmoothMin, inputType=FunctionInput_Scalar, sortPriority=32}
  n2 = FunctionInput(Preview=Constant() {r=1}) {bUsePreviewValueAsDefault=true, inputName=SmoothMax, inputType=FunctionInput_Scalar, sortPriority=33}
  n3 = Constant()
  n4 = Multiply(
   A=NamedRerouteUsage()
   B=FunctionInput(Preview=Constant() {r=5}) {bUsePreviewValueAsDefault=true, inputName=Spark Max Length, inputType=FunctionInput_Scalar, sortPriority=5}
  )
  MaterialExpressionAppendVector_10 = AppendVector(A=@n3, B=@n4)
  MaterialExpressionAppendVector_11 = AppendVector(A=@n3, B=Multiply(A=@n4) {constB=-1})
  n5 = MaterialFunctionCall(
   UVs=FunctionInput(Preview=TextureCoordinate()) {bUsePreviewValueAsDefault=true, inputName=Spark Pattern UV, inputType=FunctionInput_Vector2, sortPriority=2}
   "Rotation Angle (0-1)"=FunctionInput(Preview=Constant()) {bUsePreviewValueAsDefault=true, inputName=Rotation, inputType=FunctionInput_Scalar, sortPriority=6}
  ) {materialFunction=&CustomRotator}
  n6 = FunctionInput(Preview=ScalarParameter() {defaultValue=20, parameterName=PatternX}) {bUsePreviewValueAsDefault=true, inputName=Spark Pattern X, inputType=FunctionInput_Scalar, sortPriority=1}
  n7 = FunctionInput(Preview=ScalarParameter() {defaultValue=1, parameterName=Pattern Y}) {bUsePreviewValueAsDefault=true, inputName=Spark Pattern Y, inputType=FunctionInput_Scalar, sortPriority=2}
  n8 = Multiply(A=@n5.Rotated Values, B=AppendVector(A=@n6, B=@n7)) {constB=10}
  n9 = Frac(None=@n8)
  n10 = Time()
  n11 = Multiply(
   A=@n10
   B=FunctionInput(Preview=Constant() {r=0.5}) {bUsePreviewValueAsDefault=true, inputName=Spark Wobble Speed, inputType=FunctionInput_Scalar, sortPriority=4}
  ) {constB=0.5}
  n12 = Multiply(A=NamedRerouteUsage(), B=@n11)
  n13 = Multiply(A=Sine(None=@n12), B=Cosine(None=@n12))
  n14 = Constant() {r=0.75}
  n15 = Multiply(A=@n11, B=Multiply(A=NamedRerouteUsage(), B=Constant() {r=2})) {constB=-1}
  n16 = Multiply(A=Sine(None=@n15), B=Cosine(None=@n15))
  n17 = NamedRerouteUsage()
  n18 = Constant() {r=0.5}
  n19 = Multiply(
   A=NamedRerouteUsage()
   B=Multiply(
    A=@n10
    B=FunctionInput(Preview=@n18) {bUsePreviewValueAsDefault=true, inputName=Spark Speed, inputType=FunctionInput_Scalar, sortPriority=3}
   ) {constB=0.5}
  )
  n20 = AppendVector(
   A=LinearInterpolate(A=Multiply(A=@n13, B=@n14) {constB=0.75}, B=Multiply(A=@n14, B=@n16) {constB=0.75}, Alpha=Step(Y=@n17) {constX=0.5})
   B=ConstantBiasScale(None=Frac(None=@n19)) {bias=-0.5, scale=2}
  )
  n21 = Multiply(A=@n5.Rotated Values, B=AppendVector(A=Constant() {r=1}, B=@n7))
  n22 = Multiply(A=NamedRerouteUsage(), B=ComponentMask(None=Frac(None=@n21)) {g=true})
  n23 = AppendVector(A=Constant() {r=1}, B=Divide(A=@n7, B=@n6))
  MaterialExpressionConstantBiasScale_2 = ConstantBiasScale(None=MaterialFunctionCall(UVs=Add(A=@n9, B=@n20), Texture Scale=Multiply(A=@n22, B=@n23)) {materialFunction=&ScaleUVsByCenter}.UVs) {bias=-0.5, scale=2}
  n24 = Custom!(a=@MaterialExpressionAppendVector_10, b=@MaterialExpressionAppendVector_11, p=@MaterialExpressionConstantBiasScale_2) {
   code="float2 pa = p-a, ba = b-a;\r\nfloat h = clamp( dot(pa,ba)/dot(ba,ba), 0.0, 1.0 );\r\nreturn length( pa - ba*h );"
   description=Line
   outputType=CMOT_Float1
   inputs [
    {inputName=a, input {expression=&self:MaterialExpressionAppendVector_10}}
    {inputName=b, input {expression=&self:MaterialExpressionAppendVector_11}}
    {inputName=p, input {expression=&self:MaterialExpressionConstantBiasScale_2}}
   ]
  }
  FunctionOutput(None=OneMinus(None=SmoothStep(Min=@n1, Max=@n2, Value=@n24)))
  MaterialExpressionAdd_1 = Add(
   A=Add(A=Floor(None=@n8), B=Constant() {r=0.01}) {constB=0.01}
   B=FunctionInput(Preview=Constant()) {bUsePreviewValueAsDefault=true, inputName=Random Seed, inputType=FunctionInput_Scalar, sortPriority=7}
  )
  n25 = Custom!(x=@MaterialExpressionAdd_1) {
   code="uint2 q = 1103515245U * ( (asuint(x)>>1U) ^ (asuint((x.yx))   ) );\r\nuint  n = 1103515245U * ( (q.x  ) ^ (q.y>>3U) );\r\nreturn float(n) * (1.0/float(0xffffffffU));"
   desc=Random
   description=IQHash3
   outputType=CMOT_Float1
   inputs [{inputName=x, input {expression=&self:MaterialExpressionAdd_1}}]
  }
  NamedRerouteDeclaration(None=@n25) {name=Random, nodeColor=<rgba 0 0 0 1>}
 ]}
