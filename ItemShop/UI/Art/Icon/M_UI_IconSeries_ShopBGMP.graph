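# Default Lit material for the Icon Series item shop background, with floor parameters for metallic, specular, roughness and brightness.
# Layered radial gradients over screen-scaled UVs build a navy, cyan and purple backdrop, which is desaturated, brightened and modulated by a radial floor mask.
# Named reroutes carry foreground and background spark particles, plus animated lightbeams made from two panning, self-distorted noise layers in polar coordinates.
# Floor roughness switches for mobile shading.
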Material /ItemShop/UI/Art/Icon/M_UI_IconSeries_ShopBGMP
graph {
 properties {pixelDepthOffsetMode=PDOM_Legacy, usedShadingModels=Default Lit}
 nodes [
  Constant2Vector() {g=1080, r=1920}
  Constant() {r=0.5}
  Constant()
  n1 = MaterialFunctionCall(
   Radius=ScalarParameter() {defaultValue=0.25, parameterName=Floor-Radius Mask}
   Density=ScalarParameter() {defaultValue=1, parameterName=Floor-Density Mask}
  ) {materialFunction=&RadialGradientExponential}
  LinearInterpolate(B=ScalarParameter() {defaultValue=1, group=IsFloor, parameterName=Floor-Specular}, Alpha=@n1.RadialGradientExponential)
  MaterialFunctionCall() {materialFunction=&MF_UI_UVScaling}
  n2 = NamedRerouteUsage()
  MaterialFunctionCall(UVs=@n2, Texture Scale=Constant() {r=1.5}) {materialFunction=&ScaleUVsByCenter}
  MaterialFunctionCall!(Color, Color Brightness, Floor Radius, Floor Density, Floor Metallic, Floor Specular, Floor Roughness) {materialFunction=/Game/UI/Foundation/Environment/Shell/Materials/MaterialFunctions/MF_UI_MPItemShop_Setup}
  n3 = NamedRerouteUsage()
  n4 = Constant() {r=2}
  n5 = Constant()
  n6 = Constant()
  n7 = MaterialFunctionCall(
   UV=Add(
    A=MaterialFunctionCall(UVs=@n3, Texture Scale=@n4) {materialFunction=&ScaleUVsByCenter}.UVs
    B=AppendVector(A=@n5, B=@n6)
   )
   GradientStart=Constant()
   GradientEnd=Constant() {r=0.5}
  ) {materialFunction=&MF_UI_RadialGradient}
  n8 = Add(A=Multiply(A=NamedRerouteUsage()) {constB=0.65}, B=AppendVector(A=Constant() {r=0.4}, B=Constant() {r=0.35}))
  n9 = Constant()
  n10 = Constant() {r=1}
  n11 = LinearInterpolate(
   A=Constant3Vector() {constant=<rgba 0 0.006 0.0208 1>}
   B=Constant3Vector() {constant=<rgba 0 0.2556 0.3958 1>}
   Alpha=SmoothStep(Value=MaterialFunctionCall(UV=@n8, GradientStart=@n9, GradientEnd=@n10) {materialFunction=&MF_UI_RadialGradient}.Smooth) {constMax=2}
  )
  n12 = Constant3Vector() {constant=<rgba 0 0.6417 0.8802 1>}
  n13 = Constant() {r=-0.5}
  n14 = Constant() {r=1}
  n15 = Multiply(A=Add(A=NamedRerouteUsage(), B=AppendVector(A=@n13, B=@n14))) {constB=0.5}
  n16 = Constant()
  n17 = Constant() {r=1.5}
  n18 = Multiply(A=SmoothStep(Value=MaterialFunctionCall(UV=@n15, GradientStart=@n16, GradientEnd=@n17) {materialFunction=&MF_UI_RadialGradient}.Smooth) {constMax=4}) {constB=20}
  n19 = Add(
   A=MaterialFunctionCall(UVs=NamedRerouteUsage(), Texture Scale=Constant() {r=5}) {materialFunction=&ScaleUVsByCenter}.UVs
   B=AppendVector(A=Constant() {r=-0.1}, B=Constant() {r=-0.4})
  )
  n20 = Constant()
  n21 = Constant() {r=1}
  n22 = LinearInterpolate(
   A=LinearInterpolate(
    A=LinearInterpolate(A=@n11, B=@n12, Alpha=@n18)
    B=Constant3Vector() {constant=<rgba 0.0374 0.0138 0.138 1>}
    Alpha=MaterialFunctionCall(UV=@n19, GradientStart=@n20, GradientEnd=@n21) {materialFunction=&MF_UI_RadialGradient}.Smooth
   )
   B=Constant3Vector() {constant=<rgba 0.0036 0 0.0234 1>}
   Alpha=Multiply(A=OneMinus(None=@n7.Smooth)) {constB=1.2}
  ) {constB=0}
  n23 = Constant3Vector() {constant=<rgba 0 0.8049 1 1>}
  n24 = Multiply(A=NamedRerouteUsage()) {constB=0.3}
  n25 = NamedRerouteUsage()
  n26 = Add(
   A=LinearInterpolate(A=LinearInterpolate(A=@n22, B=@n23, Alpha=@n24), B=Constant3Vector() {constant=<rgba 0 0.2702 0.474 1>}, Alpha=Multiply(A=@n25) {constB=0.5})
   B=Divide(A=NamedRerouteUsage()) {constB=1.9}
  )
  n27 = ScalarParameter() {defaultValue=0.35, parameterName=Param_1}
  Multiply(
   A=Multiply(A=Desaturation(None=@n26, Fraction=@n27), B=ScalarParameter() {defaultValue=1.5, parameterName=FinalBrightness})
   B=LinearInterpolate(
    A=ScalarParameter() {defaultValue=1.2, parameterName=Floor-Brightness-High}
    B=ScalarParameter() {defaultValue=0.5, parameterName=Floor-Brightness-Low}
    Alpha=@n1.RadialGradientExponential
   )
  )
  n28 = MaterialFunctionCall!(ScreenUV, TextureSize, ScalePivotPoint) {materialFunction=/FortUILibrary/MaterialFunctions/Basics/MF_UI_ScreenSpaceUVScaling}
  NamedRerouteDeclaration(None=@n28.Scaled to Fill) {name=Scale to Fill UV, nodeColor=<rgba 1 0 0.2578 1>}
  NamedRerouteDeclaration(None=@n28.Scaled to Fill) {name=ScaleToFillSSUV, nodeColor=<rgba 0.0312 1 0 1>}
  NamedRerouteDeclaration(None=@n11) {name=ToriBGGrad, nodeColor=<rgba 1 0.1406 0 1>}
  NamedRerouteDeclaration(None=@n28.Scaled to Fit) {name=Scale to Fit UV, nodeColor=<rgba 0 1 0.1797 1>}
  n29 = MaterialFunctionCall(UVs=NamedRerouteUsage(), "Rotation Angle (0-1)"=Constant() {r=-0.4}) {materialFunction=&CustomRotator}
  n30 = Constant() {r=0.15}
  n31 = Constant() {r=0.5}
  n32 = Constant() {r=3}
  n33 = NamedRerouteUsage()
  n34 = Constant() {r=1}
  n35 = Min(
   A=OneMinus(None=MaterialFunctionCall!(UVs=@n29.Rotated Values, Subtract=@n30, Falloff=@n31, Intensity=@n32) {materialFunction=/Game/UI/Frontend/ItemShop/Art/_MaterialFunctions_ShopTheming/MF_UI_TopBottomGradient}.Result)
   B=MaterialFunctionCall!(
    Spark Pattern X=Constant() {r=17}
    Spark Pattern Y=Constant() {r=1}
    Spark Pattern UV=MaterialFunctionCall(UVs=@n33, Texture Scale=@n34) {materialFunction=&ScaleUVsByCenter}.UVs
    Spark Speed=Constant() {r=0.015}
    Spark Wobble Speed=Constant() {r=0.05}
    Spark Max Length=Constant()
    Rotation=Constant() {r=0.2}
    Random Seed=Constant() {r=0.02}
    SmoothMin=Constant()
    SmoothMax=Constant() {r=0.25}
   ) {materialFunction=/Game/UI/Foundation/Environment/Shell/Materials/MF_UI_ParticleSparks}.Result
  )
  n36 = NamedRerouteUsage()
  n37 = Power(Base=OneMinus(None=ComponentMask(None=@n36) {r=true}), Exp=ScalarParameter() {defaultValue=0.5, parameterName=Param_2})
  NamedRerouteDeclaration(None=Multiply(A=Saturate(None=@n35), B=@n37)) {name=Background Particles, nodeColor=<rgba 1 0 0.0469 1>}
  n38 = MaterialFunctionCall(
   UVs=MaterialFunctionCall(UVs=NamedRerouteUsage(), "Rotation Angle (0-1)"=Constant() {r=-0.4}) {materialFunction=&CustomRotator}.Rotated Values
   Subtract=Constant() {r=0.15}
   Falloff=Constant() {r=0.5}
   Intensity=Constant() {r=3}
  ) {materialFunction=&MF_UI_TopBottomGradient}
  n39 = Constant() {r=17}
  n40 = Constant() {r=1}
  n41 = Constant() {r=0.01}
  n42 = Constant()
  n43 = Constant()
  n44 = Constant() {r=0.2}
  n45 = Constant() {r=2}
  n46 = Constant()
  n47 = Constant() {r=1}
  n48 = Min(
   A=SmoothStep(Value=OneMinus(None=@n38.Result)) {constMax=1.5, constMin=-0.1}
   B=SmoothStep(Value=MaterialFunctionCall(
     Spark Pattern X=@n39
     Spark Pattern Y=@n40
     Spark Pattern UV=@n2
     Spark Speed=@n41
     Spark Wobble Speed=@n42
     Spark Max Length=@n43
     Rotation=@n44
     Random Seed=@n45
     SmoothMin=@n46
     SmoothMax=@n47
    ) {materialFunction=&MF_UI_ParticleSparks}.Result) {constMax=3}
  )
  NamedRerouteDeclaration(None=Multiply(A=Saturate(None=@n48), B=@n37)) {name=Foreground Particles, nodeColor=<rgba 0.0312 0 1 1>}
  n49 = Power(Base=NamedRerouteUsage(), Exp=Constant() {r=0.25})
  n50 = NamedRerouteUsage()
  n51 = Constant() {r=-0.42}
  n52 = OneMinus(None=MaterialFunctionCall(
    UVs=MaterialFunctionCall(UVs=@n50, "Rotation Angle (0-1)"=@n51) {materialFunction=&CustomRotator}.Rotated Values
    Subtract=Constant()
    Falloff=Constant() {r=1}
    Intensity=Constant() {r=1}
   ) {materialFunction=&MF_UI_TopBottomGradient}.Result)
  n53 = ScalarParameter() {defaultValue=1, group=Radial Gradient, parameterName=Tile}
  n54 = Multiply(A=NamedRerouteUsage(), B=@n53)
  n55 = Multiply(A=@n53) {constB=0.5}
  n56 = ScalarParameter() {defaultValue=-1.25, parameterName=offset x}
  n57 = ScalarParameter() {defaultValue=0.5, parameterName=offset y}
  n58 = MaterialFunctionCall(
   Vector or UVs=Add(A=Subtract(A=@n54, B=@n55) {constB=3}, B=AppendVector(A=@n56, B=@n57)) {constB=0.5}
   Swizzle Coordinate Output=StaticBool() {value=true}
  ) {materialFunction=&VectorToRadialValue}
  n59 = Time()
  n60 = ScalarParameter() {defaultValue=0.1, parameterName=TimeFactor2}
  n61 = Multiply(A=Panner(Coordinate=@n58.Vector Converted to Angle, Time=Multiply(A=@n59, B=@n60)) {speedX=0.111, speedY=0.1}) {constB=2}
  n62 = TextureSample(UVs=Multiply(
    A=TextureSample(UVs=@n61) {samplerType=SAMPLERTYPE_Grayscale, texture=/Game/UI/Frontend/ItemShop/Art/Rumble/LowResBlurredNoise_Rumble}.R
    B=ScalarParameter() {defaultValue=0.0125, parameterName=DeformIntensity}
   )) {samplerType=SAMPLERTYPE_Grayscale, texture=&LowResBlurredNoise_Rumble}
  n63 = Constant() {r=3}
  n64 = Saturate(None=Multiply(A=Power(Base=@n62.R, Exp=@n63) {constExponent=1}, B=Constant() {r=1.5}) {constB=2})
  n65 = Time()
  n66 = ScalarParameter() {defaultValue=0.1, parameterName=TimeFactor}
  n67 = Multiply(A=Panner(Coordinate=@n58.Vector Converted to Angle, Time=Multiply(A=@n65, B=@n66)) {speedX=0.025, speedY=-0.025}) {constB=2}
  n68 = TextureSample(UVs=Multiply(
    A=TextureSample(UVs=@n67) {samplerType=SAMPLERTYPE_Grayscale, texture=&LowResBlurredNoise_Rumble}.R
    B=ScalarParameter() {defaultValue=0.25, parameterName=DeformIntensity_01}
   )) {samplerType=SAMPLERTYPE_Grayscale, texture=&LowResBlurredNoise_Rumble}
  n69 = Constant() {r=3}
  n70 = Saturate(None=Multiply(A=Power(Base=@n68.R, Exp=@n69) {constExponent=1}, B=Constant() {r=1.5}) {constB=2})
  n71 = Min(A=SmoothStep(Value=@n52) {constMax=2}, B=Saturate(None=Min(A=@n64, B=@n70)))
  NamedRerouteDeclaration(None=Multiply(A=Constant3Vector() {constant=<rgba 0 0.2556 0.3958 1>}, B=Multiply(A=@n49, B=@n71))) {name=Lightbeams, nodeColor=<rgba 1 0.8203 0 1>}
  NamedRerouteUsage()
  NamedRerouteUsage()
  Saturate(None=@n52)
  ScalarParameter() {defaultValue=-1.25, parameterName=offset x}
  ScalarParameter() {defaultValue=1, group=IsFloor, parameterName=Floor-Metallic}
  ScalarParameter() {defaultValue=0.5, parameterName=offset y}
  ShadingPathSwitch(
   Default=ScalarParameter() {defaultValue=0.25, group=IsFloor, parameterName=Floor-Roughness}
   Mobile=ScalarParameter() {defaultValue=0.75, parameterName=Floor-Roughness_mobile}
  )
  TextureCoordinate()
  TextureCoordinate()
 ]
}
